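# Post-style animation Blueprint for the M_MED_Lasso_Polo male body cosmetic, deforming it on top of the pose copied from its attached parent mesh.
# The copied pose runs through a twists Control Rig, then a sculpters Control Rig with knee corrective inputs, to the output.
# Spring-driven AnimDynamics on both sub_thigh bones forms a separate chain that does not reach the output.
# The event graph holds only an empty update event.
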AnimBlueprint /BRCosmetics/Characters/Player/Male/Medium/Bodies/M_MED_Lasso_Polo/Meshes/M_MED_Lasso_Polo_AnimBP
target_skeleton=&/M_MED_Lasso_Polo_Skeleton
graphs [
 AnimGraph {nodes [
   n1 = Anim Dynamics - Bone: sub_thigh_l(ComponentPose=Local To Component(), Alpha=1, ExternalForce="0,0,0") {
    angularDampingOverride=1
    angularSpringConstant=1500
    bAngularSpring=true
    bGravityOverrideInSimSpace=true
    bOverrideAngularDamping=true
    gravityScale=0.35
    simulationSpace=RootRelative
    alphaBoolBlend {blendOption=Linear}
    boundBone {boneName=sub_thigh_l}
    physicsBodyDefinitions [{
      sphereCollisionRadius=0
      boundBone {boneName=sub_thigh_l}
      boxExtents=<10 20 20>
      constraintSetup {angularLimitsMax=<0 6 6>, angularLimitsMin=<0 -6 -6>, angularTarget=<0 0 0>}
      localJointOffset=<-5 0 0.65>
     }]
   }
   n2 = Anim Dynamics - Bone: sub_thigh_r(ComponentPose=@n1, Alpha=1, ExternalForce="0,0,0") {
    angularDampingOverride=1
    angularSpringConstant=1500
    bAngularSpring=true
    bGravityOverrideInSimSpace=true
    bOverrideAngularDamping=true
    gravityScale=0.35
    simulationSpace=RootRelative
    alphaBoolBlend {blendOption=Linear}
    boundBone {boneName=sub_thigh_r}
    physicsBodyDefinitions [{
      sphereCollisionRadius=0
      boundBone {boneName=sub_thigh_r}
      boxExtents=<10 20 20>
      constraintSetup {angularLimitsMax=<0 6 6>, angularLimitsMin=<0 -6 -6>, angularTarget=<0 0 0>}
      localJointOffset=<5 0 -0.65>
     }]
   }
   Component To Local(ComponentPose=@n2)
   n3 = Control Rig(Source=Copy Pose From Mesh() {bUseAttachedParent=true}) {
    alphaBoolBlend {blendOption=Linear}
    controlRigAssetReference {blueprintRigClass=&/M_MED_Lasso_Polo_Deform_CtrlRig_twists.M_MED_Lasso_Polo_Deform_CtrlRig_twists_C}
   }
   n4 = Control Rig!(
    Source=@n3
    Alpha
    KneeRot_fr_l_114deg:Vector
    KneeTrans_fr_l_114deg:Vector="0.258,-3.591,0.009"
    KneeScale_fr_l_114deg:Vector="1,1,1"
    KneeScale_fr_l_130deg:Vector="1,1,1"
    KneeRot_fr_l_130deg:Vector="0,0,-51.2"
    KneeTrans_fr_l_130deg:Vector="0.258,-3.591,0.009"
   ) -> (Pose) {
    alphaBoolBlend {blendOption=Linear}
    controlRigAssetReference {blueprintRigClass=&/M_MED_Lasso_Polo_Deform_CtrlRig_sculpters.M_MED_Lasso_Polo_Deform_CtrlRig_sculpters_C}
   }
   Output Pose(Result=@n4) {layerGroup=None}
  ]}
 EventGraph {nodes [Event BlueprintUpdateAnimation()]}
]
variables []
